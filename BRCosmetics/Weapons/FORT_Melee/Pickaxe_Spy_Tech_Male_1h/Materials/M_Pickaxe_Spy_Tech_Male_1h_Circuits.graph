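# Translucent, two-sided unlit material that draws glowing, scrolling circuit traces on the Spy Tech male one-handed pickaxe.
# Planar coordinates from pre-skinned position drive two panned, tinted circuit-board samples, one offset by BumpOffset for parallax depth.
# These are summed with an amber sphere-mask glow around a pivot parameter and passed through emissive adjustments.
# A rotated mesh-UV circuit mask, modulated by panning noise and that glow, masks the result.
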Material /BRCosmetics/Weapons/FORT_Melee/Pickaxe_Spy_Tech_Male_1h/Materials/M_Pickaxe_Spy_Tech_Male_1h_Circuits
graph {
 properties {
  bUsedWithMeshDeformer=true
  bUsedWithSkeletalMesh=true
  blendMode=BLEND_Translucent
  pixelDepthOffsetMode=PDOM_Legacy
  shadingModel=MSM_Unlit
  twoSided=true
  usedShadingModels=Unlit
 }
 nodes [
  Constant3Vector()
  Constant3Vector() {constant=<rgba 0 0 1 0>}
  Constant()
  Constant() {r=1}
  DepthFade()
  n1 = MaterialFunctionCall(Float3=PreSkinnedPosition()) {materialFunction=&BreakOutFloat3Components}
  n2 = AppendVector(A=@n1.R, B=@n1.R)
  n3 = Constant() {r=0.25}
  n4 = VertexInterpolator(VS=AppendVector(
    A=MaterialFunctionCall(UVs=@n2, "Rotation Angle (0-1)"=@n3) {materialFunction=&CustomRotator}.Rotated Values
    B=@n1.B
   ))
  n5 = Abs(None=Constant() {r=48})
  n6 = ComponentMask(None=Divide(A=@n4.PS, B=Multiply(A=@n5) {constB=-1})) {b=true, g=true}
  n7 = ScalarParameter() {defaultValue=-0.65, parameterName=Circuit Panning}
  n8 = Constant()
  n9 = AppendVector(A=Add(A=@n7, B=ScalarParameter() {defaultValue=0.001, parameterName=ggasfasf}), B=@n8)
  n10 = TextureSample(UVs=BumpOffset(
    Coordinate=Panner(Coordinate=@n6, Speed=@n9) {speedY=1}
    Height=ScalarParameter() {defaultValue=1.1, group=Parallax, parameterName=Parallax Height}
   )) {samplerType=SAMPLERTYPE_Masks, texture=/BRCosmetics/Weapons/FORT_Melee/Pickaxe_Spy_Tech_Male_1h/Textures/T_Spy_Tech_1H_Circuit_Board_Alpha}
  n11 = Multiply(A=VectorParameter() {parameterName=fhasfasdf, defaultValue=<rgba 0.2604 0.1153 0 1>}.RGB, B=@n10.R)
  n12 = TextureSample(UVs=Panner(Coordinate=@n6, Speed=AppendVector(A=@n7, B=@n8)) {speedY=1}) {samplerType=SAMPLERTYPE_Masks, texture=&T_Spy_Tech_1H_Circuit_Board_Alpha}
  n13 = Multiply(A=@n12.R, B=VectorParameter() {parameterName=hagasf, defaultValue=<rgba 1 0.7643 0.0573 1>}.RGB)
  n14 = Fresnel(ExponentIn=ScalarParameter() {defaultValue=0.2, parameterName=Parallax Opacity Fresnel}) {baseReflectFraction=0}
  n15 = ScalarParameter() {defaultValue=1, group=Parallax, parameterName=Parallax Global Opacity}
  LinearInterpolate(A=@n11, B=@n13, Alpha=Saturate(None=Multiply(A=@n14, B=@n15)))
  n16 = SphereMask(
   A=@n4.PS
   B=ComponentMask(None=VectorParameter() {parameterName=Pre-Skinned Pivot Offset, defaultValue=<rgba 1.136 0 40 0>}.RGB) {b=true, g=true, r=true}
   Radius=Constant() {r=75}
   Hardness=Constant()
  ) {hardnessPercent=30}
  n17 = Constant() {r=3.5}
  n18 = Multiply(
   A=LinearInterpolate(A=Constant3Vector() {constant=<rgba 0 0 0 1>}, B=Constant3Vector() {constant=<rgba 1.05 0.665 0 1>}, Alpha=Power(Base=@n16, Exp=@n17))
   B=Constant() {r=0.85}
  )
  n19 = Add(A=Saturate(None=Add(A=@n11, B=@n13)), B=@n18)
  MaterialFunctionCall(Emissive=@n19) {materialFunction=&MF_FX_EmissiveAdjustments}
  n20 = MaterialFunctionCall(UVs=TextureCoordinate(), "Rotation Angle (0-1)"=Constant() {r=0.75}) {materialFunction=&CustomRotator}
  n21 = ComponentMask(None=@n20.Rotated Values) {r=true}
  n22 = Constant() {r=3.6}
  n23 = ComponentMask(None=@n20.Rotated Values) {g=true}
  n24 = Constant() {r=1}
  n25 = TextureSample(UVs=AppendVector(A=Multiply(A=@n21, B=@n22), B=Multiply(A=@n23, B=@n24))) {samplerType=SAMPLERTYPE_Masks, texture=&T_Spy_Tech_1H_Circuit_Board_Alpha}
  n26 = Constant() {r=1.1}
  n27 = Constant() {r=-0.5}
  n28 = Constant() {r=0.35}
  n29 = TextureSample!(UVs=Panner(Coordinate=Multiply(A=@n6, B=@n26), Speed=AppendVector(A=@n27, B=@n28)) {speedY=1}, Tex, Bias=Constant(), Apply View MipBias) {mipValueMode=TMVM_MipBias, texture=/Game/Athena/UI/Challenges/Art/T_UI_DataChip_Noise}
  n30 = Power(Base=Add(A=@n18, B=Power(Base=@n25.G, Exp=@n29.R)), Exp=Constant() {r=5})
  Multiply(A=Saturate(None=@n19), B=Multiply(A=@n25.G, B=Saturate(None=@n30)))
 ]
}
